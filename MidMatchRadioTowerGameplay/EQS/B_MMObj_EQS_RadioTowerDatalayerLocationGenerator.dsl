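; EQS generator that supplies locations of mid-match radio towers that can still be activated, for use in environment queries.
; It gets the Athena game state's radio tower component and, if both are valid, reads its pending datalayer locations.
; The list is shuffled, then each location is added as a generated vector; the input context locations are not used.
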(fn DoItemGeneration
  (params (ContextLocations :type Array%20of%20Vectors))
  (bind _returnvalue (Fortnite|Athena|GameState|GetGameStateAthena))
  (bind _returnvalue_2 (Actor|GetComponentbyClass _returnvalue "/MidMatchRadioTowerGameplay/Gameplay/B_MMObj_RadioTowerGameStateComponent.B_MMObj_RadioTowerGameStateComponent_C"))
  ;[0.1719 0.03759 0.1431] Add locations of radio towers that can still be activated.
  (Utilities|FlowControl|Sequence
    (:then_0
      (Utilities|IsValid _returnvalue
        (:Is%20Valid
          (Utilities|IsValid _returnvalue_2
            (:Is%20Valid
              (bind _locations (EQS|GetRadioTowerDatalayerLocationsPending _returnvalue_2))
              (Utilities|Array|Shuffle _locations)
              (for _array_element _locations
                (EQS|AddGeneratedVector _array_element)))))))
    (:then_1
      (return))))
