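# EQS query that picks a location for a mid-match Radio Tower data layer, keeping it away from towers already activated.
# A Blueprint generator supplies candidate data layer locations as points.
# A 2D distance test against the activated tower locations context filters out points closer than a minimum bound to a Fortnite scalable-float provider.
# The same test scores linearly, favoring candidates farther from activated towers.
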EnvQuery /MidMatchRadioTowerGameplay/EQS/EQS_RadioTowerDatalayerSelection
options [{
  generator {
   name=B_MMObj_EQS_RadioTowerDatalayerLocationGenerator_C_0
   class=B_MMObj_EQS_RadioTowerDatalayerLocationGenerator_C
   properties {
    bAutoSortTests=true
    bCanRunAsync=false
    context=&EnvQueryContext_Querier
    envQueryResultNormalizationOption=Default
    generatedItemType=/Script/AIModule.EnvQueryItemType_Point
    generatorsActionDescription=""
    optionName=""
   }
  }
  tests [EnvQueryTest_Distance_0 {
    class=EnvQueryTest_Distance
    properties {
     distanceTo=&/B_MMObj_EQC_ActivatedRadioTowerDatalayerLocations.B_MMObj_EQC_ActivatedRadioTowerDatalayerLocations_C
     filterType=Minimum
     scoringEquation=Linear
     scoringFactor=1
     testComment=""
     testMode=Distance2D
     testPurpose=FilterAndScore
     floatValueMin {
      dataBinding=&self:EnvQueryTest_Distance_0.FortAIDataProvider_ScalableFloat_0
      dataField=FloatValue
      defaultValue=0
     }
    }
   }]
 }]
